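# Animation Blueprint for the Fortnite player driving a golf cart, output through the VehicleFinalPoseOverride layer.
# A Locomotion state machine driven by GolfCartInput covers idle, drive, reverse, brake, boost and powerslide, with steering leans and gender-specific powerslide clips.
# Its pose gains an idle additive, head aim (off while reversing) and the FullBody slot.
# Pelvis impact displacement, foot IK and socket-target hand IK follow, then weapon_r scales to zero.
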AnimBlueprint /Game/Animation/Game/MainPlayer/Vehicles/GolfCart/GolfCartDriver_AnimBP
target_skeleton=&Fortnite_M_Avg_Player_Skeleton
graphs [
 AnimGraph {nodes [Output Pose() {layerGroup=None}]}
 EventGraph {nodes [Event BlueprintUpdateAnimation()]}
 OnPedal {nodes [OnPedal() {is=title K2Node_FunctionEntry}]}
 VehicleFinalPoseOverride {nodes [
   Input Pose() {name=InFinalPose}
   AnimGraphNode_StateMachine_0 = Locomotion() {is=title AnimGraphNode_StateMachine}
   n1 = Get GolfCartInput() -> (GolfCartInput:Fort Anim Input Golf Cart Structure)
   n2 = AimOffset Player 'Driver_Head_AO_GolfCart'!(
    X:float=Get AimYaw() -> (AimYaw:float)
    Y:float=Get AimPitch()
    BasePose:Pose Link Structure=Apply Additive(
     Base=@AnimGraphNode_StateMachine_0
     Additive="Sequence Player (additive) 'Idle_Driver_GolfCart'"() {groupName=Idle, method=SyncGroup, sequence=&/Idle_Driver_GolfCart}
    ) {alphaBoolBlend {blendOption=Linear}}
    bAlphaBoolEnabled:bool=NOT Boolean(A=Break Fort Anim Input Golf Cart!(FortAnimInput_GolfCart:"Fort Anim Input Golf Cart Structure (by ref)"=@n1) -> (
      bIsUsingGolfCart:bool=~
      bIsDriver:bool=~
      bIsFrontPassenger:bool=~
      bIsBackLeftPassenger:bool=~
      bIsBackRightPassenger:bool=~
      bIsFrontPassengerAndLeaning:bool=~
      bIsBackPassengerAndLeaning:bool=~
      bIsDrifting:bool=~
      bIsBoosting:bool=~
      bIsEBraking:bool=~
      bIsReversing:bool=~
      bIsBraking:bool=~
      bIsMoving:bool=~
      bIsMovingForward:bool=~
      bIsPowerSliding:bool=~
      bIsLeaning:bool=~
      bIsLeaningOrBouncing:bool=~
      bIsBounceCrouching:bool=~
      bIsBounceCrouched:bool=~
      bIsBounceJumping:bool=~
      bIsBounceRecoiling:bool=~
      bIsSteeringRight:bool=~
      bIsSteeringLeft:bool=~
      RunForwardAlpha:float=~
      BounceCompression:float=~
      LeanPosition:Vector=~
      LeanPositionX:float=~
      LeanPositionY:float=~
      LeanPositionZ:float=~
      bAimFWD:bool=~
      bAimBWD:bool=~
      bAimLFT:bool=~
      bAimRGT:bool=~
      PawnToVehicleDeltaYawAngleDegrees:float=~
      AimCardDirDeadZoneAngleDegrees:float=~
      AimCardDirAngleOffsetDegrees:float=~
      AimFWDDeltaAngleDegrees:float=~
      AimBWDDeltaAngleDegrees:float=~
      AimLFTDeltaAngleDegrees:float=~
      AimRGTDeltaAngleDegrees:float=~
      SlopePitchDegreeAngle:float=~
      SlopeRollDegreeAngle:float=~
      SteerAngle:float=~
     ) {is=title K2Node_BreakStruct}.bIsReversing)
   ) -> (Pose:Pose Link Structure) {
    alphaInputType=Bool
    blendSpace=/Game/Animation/Game/MainPlayer/Vehicles/GolfCart/AO/Driver_Head_AO_GolfCart
    alphaBoolBlend {blendOption=Linear}
   }
   n3 = Slot 'FullBody'(Source=@n2) {slotName=FullBody}
   n4 = Leg IK!(
    ComponentPose:Component Space Pose Link Structure="Transform (Modify) Bone - Bone: pelvis"(
     ComponentPose=Local To Component(LocalPose=@n3)
     Alpha=1
     Translation=Get ImpactDisplacement() -> (ImpactDisplacement:Vector)
     Rotation="0,0,0"
     Scale="1,1,1"
    ) {translationMode=BMM_Additive, alphaBoolBlend {blendOption=Linear}, boneToModify {boneName=pelvis}}
    Alpha:float=1
   ) -> (Pose:Component Space Pose Link Structure) {
    is=title AnimGraphNode_LegIK
    alphaBoolBlend {blendOption=Linear}
    legsDefinition [
     {
      bEnableKneeTwistCorrection=false
      bEnableRotationLimit=false
      footBoneForwardAxis=Y
      hingeRotationAxis=None
      minRotationAngle=15
      numBonesInLimb=2
      twistOffsetCurveName=None
      fKFootBone {boneName=foot_l}
      iKFootBone {boneName=ik_foot_l}
     }
     {
      bEnableKneeTwistCorrection=false
      bEnableRotationLimit=false
      footBoneForwardAxis=Y
      hingeRotationAxis=None
      minRotationAngle=15
      numBonesInLimb=2
      twistOffsetCurveName=None
      fKFootBone {boneName=foot_r}
      iKFootBone {boneName=ik_foot_r}
     }
    ]
   }
   n5 = Break Fort Anim Input Socket Based IKTarget!(FortAnimInput_SocketBasedIKTarget:"Fort Anim Input Socket Based IKTarget Structure (by ref)"=Get LeftHandIKTarget() -> (LeftHandIKTarget:Fort Anim Input Socket Based IKTarget Structure)) -> (SocketName:Name, AlphaCurveName:Name, TargetRotation:Rotator=~, TargetLocation:Vector=~, Alpha:float=~) {is=title K2Node_BreakStruct}
   n6 = "Transform (Modify) Bone - Bone: ik_hand_l"!(ComponentPose=@n4, Alpha=@n5.Alpha, Translation=@n5.TargetLocation, Rotation=@n5.TargetRotation) -> (Pose) {
    rotationMode=BMM_Replace
    rotationSpace=BCS_WorldSpace
    translationMode=BMM_Replace
    translationSpace=BCS_WorldSpace
    alphaBoolBlend {blendOption=Linear}
    boneToModify {boneName=ik_hand_l}
   }
   n7 = Break Fort Anim Input Socket Based IKTarget(FortAnimInput_SocketBasedIKTarget=Get RightHandIKTarget() -> (RightHandIKTarget:Fort Anim Input Socket Based IKTarget Structure)) -> (TargetRotation=~, TargetLocation=~, Alpha=~)
   n8 = "Transform (Modify) Bone - Bone: ik_hand_r"!(ComponentPose=@n6, Alpha=@n7.Alpha, Translation=@n7.TargetLocation, Rotation=@n7.TargetRotation) -> (Pose) {
    rotationMode=BMM_Replace
    rotationSpace=BCS_WorldSpace
    translationMode=BMM_Replace
    translationSpace=BCS_WorldSpace
    alphaBoolBlend {blendOption=Linear}
    boneToModify {boneName=ik_hand_r}
   }
   n9 = Leg IK(ComponentPose=@n8, Alpha=1) {
    alphaBoolBlend {blendOption=Linear}
    legsDefinition [{
      bEnableKneeTwistCorrection=true
      bEnableRotationLimit=false
      footBoneForwardAxis=Y
      hingeRotationAxis=None
      minRotationAngle=15
      numBonesInLimb=2
      twistOffsetCurveName=None
      fKFootBone {boneName=hand_l}
      iKFootBone {boneName=ik_hand_l}
     }]
   }
   n10 = Leg IK(ComponentPose=@n9, Alpha=1) {
    alphaBoolBlend {blendOption=Linear}
    legsDefinition [{
      bEnableKneeTwistCorrection=true
      bEnableRotationLimit=false
      footBoneForwardAxis=Y
      hingeRotationAxis=None
      minRotationAngle=15
      numBonesInLimb=2
      twistOffsetCurveName=None
      fKFootBone {boneName=hand_r}
      iKFootBone {boneName=ik_hand_r}
     }]
   }
   n11 = "Transform (Modify) Bone - Bone: weapon_r"!(ComponentPose=@n10, Alpha=1, Scale="0,0,0") -> (Pose) {
    lODThreshold=3
    scaleMode=BMM_Replace
    scaleSpace=BCS_BoneSpace
    alphaBoolBlend {blendOption=Linear}
    boneToModify {boneName=weapon_r}
    scale=<0 0 0>
   }
   Output Pose(Result=Component To Local(ComponentPose=@n11)) {layerGroup=VehicleOverride}
   AnimGraphNode_StateMachine_1 = Driver_Idle() {is=title AnimGraphNode_StateMachine}
   Save cached pose 'Idle_Pose'(Pose=@AnimGraphNode_StateMachine_1) {cacheName=Idle_Pose}
  ]}
 Locomotion {
  entry=Default
  path=VehicleFinalPoseOverride.AnimGraphNode_StateMachine_0.Locomotion
  ids=id
  nodes [
   AnimStateNode_0 = Default() {is=title AnimStateNode}
   AnimStateNode_1 = Drive_N_Start_GolfCart() {is=title AnimStateNode}
   AnimStateNode_12 = Boost_Loop() {is=title AnimStateNode}
   AnimStateNode_16 = Powersliding_Start() {is=title AnimStateNode}
   AnimStateNode_17 = Powersliding_Loop() {is=title AnimStateNode}
   AnimStateNode_18 = Powersliding_End() {is=title AnimStateNode}
   AnimStateNode_2 = Drive_N_Loop_GolfCart() {is=title AnimStateNode}
   AnimStateNode_3 = Reverse() {is=title AnimStateNode}
   AnimStateNode_4 = Braking() {is=title AnimStateNode}
   AnimStateNode_5 = Back_to_Idle_Transition() {is=title AnimStateNode}
   AnimStateNode_6 = Reverse_Start() {is=title AnimStateNode}
   AnimStateNode_7 = Boost_Start() {is=title AnimStateNode}
   AnimStateNode_8 = Reverse_End() {is=title AnimStateNode}
   AnimStateTransitionNode_0 = AnimStateTransitionNode() {crossfadeDuration=0.35, priorityOrder=2}
   AnimStateTransitionNode_1 = AnimStateTransitionNode() {bAutomaticRuleBasedOnSequencePlayerInState=true, crossfadeDuration=0.35}
   AnimStateTransitionNode_10 = AnimStateTransitionNode() {bAutomaticRuleBasedOnSequencePlayerInState=true, crossfadeDuration=0.35}
   AnimStateTransitionNode_11 = AnimStateTransitionNode() {crossfadeDuration=0.1}
   AnimStateTransitionNode_12 = AnimStateTransitionNode() {bAutomaticRuleBasedOnSequencePlayerInState=true, priorityOrder=2}
   AnimStateTransitionNode_13 = AnimStateTransitionNode() {bAutomaticRuleBasedOnSequencePlayerInState=true, crossfadeDuration=0.35}
   AnimStateTransitionNode_14 = AnimStateTransitionNode() {crossfadeDuration=0.1}
   AnimStateTransitionNode_15 = AnimStateTransitionNode()
   AnimStateTransitionNode_16 = AnimStateTransitionNode() {crossfadeDuration=0.1}
   AnimStateTransitionNode_17 = AnimStateTransitionNode() {bAutomaticRuleBasedOnSequencePlayerInState=true}
   AnimStateTransitionNode_18 = AnimStateTransitionNode() {crossfadeDuration=0.1}
   AnimStateTransitionNode_19 = AnimStateTransitionNode() {crossfadeDuration=0.1}
   AnimStateTransitionNode_2 = AnimStateTransitionNode() {crossfadeDuration=0.1}
   AnimStateTransitionNode_20 = AnimStateTransitionNode() {bAutomaticRuleBasedOnSequencePlayerInState=true, crossfadeDuration=0.4}
   AnimStateTransitionNode_21 = AnimStateTransitionNode() {crossfadeDuration=0.35}
   AnimStateTransitionNode_23 = AnimStateTransitionNode() {crossfadeDuration=0.1}
   AnimStateTransitionNode_3 = AnimStateTransitionNode() {priorityOrder=2}
   AnimStateTransitionNode_4 = AnimStateTransitionNode() {bAutomaticRuleBasedOnSequencePlayerInState=true}
   AnimStateTransitionNode_5 = AnimStateTransitionNode()
   AnimStateTransitionNode_6 = AnimStateTransitionNode()
   AnimStateTransitionNode_7 = AnimStateTransitionNode()
   AnimStateTransitionNode_8 = AnimStateTransitionNode()
   AnimStateTransitionNode_9 = AnimStateTransitionNode() {bAutomaticRuleBasedOnSequencePlayerInState=true, crossfadeDuration=0.3}
  ]
  links [
   {
    dst=Default
    rule=VehicleFinalPoseOverride.AnimGraphNode_StateMachine_0.Locomotion.AnimStateTransitionNode_9.Transition
    src=Back_to_Idle_Transition
    via=AnimStateTransitionNode_9
   }
   {
    dst=Drive_N_Loop_GolfCart
    rule=VehicleFinalPoseOverride.AnimGraphNode_StateMachine_0.Locomotion.AnimStateTransitionNode_21.Transition
    src=Boost_Loop
    via=AnimStateTransitionNode_21
   }
   {
    dst=Boost_Loop
    rule=VehicleFinalPoseOverride.AnimGraphNode_StateMachine_0.Locomotion.AnimStateTransitionNode_20.Transition
    src=Boost_Start
    via=AnimStateTransitionNode_20
   }
   {
    dst=Powersliding_Start
    rule=VehicleFinalPoseOverride.AnimGraphNode_StateMachine_0.Locomotion.AnimStateTransitionNode_14.Transition
    src=Boost_Start
    via=AnimStateTransitionNode_14
   }
   {
    dst=Drive_N_Loop_GolfCart
    rule=VehicleFinalPoseOverride.AnimGraphNode_StateMachine_0.Locomotion.AnimStateTransitionNode_10.Transition
    src=Braking
    via=AnimStateTransitionNode_10
   }
   {
    dst=Powersliding_Start
    rule=VehicleFinalPoseOverride.AnimGraphNode_StateMachine_0.Locomotion.AnimStateTransitionNode_23.Transition
    src=Braking
    via=AnimStateTransitionNode_23
   }
   {
    dst=Drive_N_Start_GolfCart
    rule=VehicleFinalPoseOverride.AnimGraphNode_StateMachine_0.Locomotion.AnimStateTransitionNode_0.Transition
    src=Default
    via=AnimStateTransitionNode_0
   }
   {
    dst=Reverse_Start
    rule=VehicleFinalPoseOverride.AnimGraphNode_StateMachine_0.Locomotion.AnimStateTransitionNode_11.Transition
    src=Default
    via=AnimStateTransitionNode_11
   }
   {
    dst=Back_to_Idle_Transition
    rule=VehicleFinalPoseOverride.AnimGraphNode_StateMachine_0.Locomotion.AnimStateTransitionNode_8.Transition
    src=Drive_N_Loop_GolfCart
    via=AnimStateTransitionNode_8
   }
   {
    dst=Boost_Start
    rule=VehicleFinalPoseOverride.AnimGraphNode_StateMachine_0.Locomotion.AnimStateTransitionNode_15.Transition
    src=Drive_N_Loop_GolfCart
    via=AnimStateTransitionNode_15
   }
   {
    dst=Braking
    rule=VehicleFinalPoseOverride.AnimGraphNode_StateMachine_0.Locomotion.AnimStateTransitionNode_5.Transition
    src=Drive_N_Loop_GolfCart
    via=AnimStateTransitionNode_5
   }
   {
    dst=Powersliding_Start
    rule=VehicleFinalPoseOverride.AnimGraphNode_StateMachine_0.Locomotion.AnimStateTransitionNode_6.Transition
    src=Drive_N_Loop_GolfCart
    via=AnimStateTransitionNode_6
   }
   {
    dst=Reverse_Start
    rule=VehicleFinalPoseOverride.AnimGraphNode_StateMachine_0.Locomotion.AnimStateTransitionNode_2.Transition
    src=Drive_N_Loop_GolfCart
    via=AnimStateTransitionNode_2
   }
   {
    dst=Drive_N_Loop_GolfCart
    rule=VehicleFinalPoseOverride.AnimGraphNode_StateMachine_0.Locomotion.AnimStateTransitionNode_1.Transition
    src=Drive_N_Start_GolfCart
    via=AnimStateTransitionNode_1
   }
   {
    dst=Drive_N_Loop_GolfCart
    rule=VehicleFinalPoseOverride.AnimGraphNode_StateMachine_0.Locomotion.AnimStateTransitionNode_17.Transition
    src=Powersliding_End
    via=AnimStateTransitionNode_17
   }
   {
    dst=Powersliding_End
    rule=VehicleFinalPoseOverride.AnimGraphNode_StateMachine_0.Locomotion.AnimStateTransitionNode_16.Transition
    src=Powersliding_Loop
    via=AnimStateTransitionNode_16
   }
   {
    dst=Powersliding_Loop
    rule=VehicleFinalPoseOverride.AnimGraphNode_StateMachine_0.Locomotion.AnimStateTransitionNode_13.Transition
    src=Powersliding_Start
    via=AnimStateTransitionNode_13
   }
   {
    dst=Drive_N_Start_GolfCart
    rule=VehicleFinalPoseOverride.AnimGraphNode_StateMachine_0.Locomotion.AnimStateTransitionNode_7.Transition
    src=Reverse
    via=AnimStateTransitionNode_7
   }
   {
    dst=Powersliding_Start
    rule=VehicleFinalPoseOverride.AnimGraphNode_StateMachine_0.Locomotion.AnimStateTransitionNode_18.Transition
    src=Reverse
    via=AnimStateTransitionNode_18
   }
   {
    dst=Reverse_End
    rule=VehicleFinalPoseOverride.AnimGraphNode_StateMachine_0.Locomotion.AnimStateTransitionNode_3.Transition
    src=Reverse
    via=AnimStateTransitionNode_3
   }
   {
    dst=Default
    rule=VehicleFinalPoseOverride.AnimGraphNode_StateMachine_0.Locomotion.AnimStateTransitionNode_4.Transition
    src=Reverse_End
    via=AnimStateTransitionNode_4
   }
   {
    dst=Powersliding_Start
    rule=VehicleFinalPoseOverride.AnimGraphNode_StateMachine_0.Locomotion.AnimStateTransitionNode_19.Transition
    src=Reverse_Start
    via=AnimStateTransitionNode_19
   }
   {
    dst=Reverse
    rule=VehicleFinalPoseOverride.AnimGraphNode_StateMachine_0.Locomotion.AnimStateTransitionNode_12.Transition
    src=Reverse_Start
    via=AnimStateTransitionNode_12
   }
  ]
 }
 Default {
  path=VehicleFinalPoseOverride.AnimGraphNode_StateMachine_0.Locomotion.AnimStateNode_0.Default
  nodes [Output Animation Pose(Result=Use cached pose 'Idle_Pose'())]
 }
 Drive_N_Start_GolfCart {
  path=VehicleFinalPoseOverride.AnimGraphNode_StateMachine_0.Locomotion.AnimStateNode_1.Drive_N_Start_GolfCart
  nodes [Output Animation Pose(Result=Apply Mesh Space Additive!(
     Base:Pose Link Structure=Use cached pose 'Idle_Pose'()
     Additive:Pose Link Structure="Sequence Player (additive) 'Drive_N_Start_GolfCart'"() {bLoopAnimation=false, sequence=&/Drive_N_Start_GolfCart}
     Alpha:float=1
    ) -> (Pose:Pose Link Structure) {is=title AnimGraphNode_ApplyMeshSpaceAdditive, alphaBoolBlend {blendOption=Linear}})]
 }
 Boost_Loop {
  path=VehicleFinalPoseOverride.AnimGraphNode_StateMachine_0.Locomotion.AnimStateNode_12.Boost_Loop
  nodes [
   n1 = AimOffset Player 'DriveForwardLeans_AO_GolfCart'!(
    X:float=Break Fort Anim Input Golf Cart(FortAnimInput_GolfCart=Get GolfCartInput()) -> (
     bIsUsingGolfCart=~
     bIsDriver=~
     bIsFrontPassenger=~
     bIsBackLeftPassenger=~
     bIsBackRightPassenger=~
     bIsFrontPassengerAndLeaning=~
     bIsBackPassengerAndLeaning=~
     bIsDrifting=~
     bIsBoosting=~
     bIsEBraking=~
     bIsReversing=~
     bIsBraking=~
     bIsMoving=~
     bIsMovingForward=~
     bIsPowerSliding=~
     bIsLeaning=~
     bIsLeaningOrBouncing=~
     bIsBounceCrouching=~
     bIsBounceCrouched=~
     bIsBounceJumping=~
     bIsBounceRecoiling=~
     bIsSteeringRight=~
     bIsSteeringLeft=~
     RunForwardAlpha=~
     BounceCompression=~
     LeanPosition=~
     LeanPositionX=~
     LeanPositionY=~
     LeanPositionZ=~
     bAimFWD=~
     bAimBWD=~
     bAimLFT=~
     bAimRGT=~
     PawnToVehicleDeltaYawAngleDegrees=~
     AimCardDirDeadZoneAngleDegrees=~
     AimCardDirAngleOffsetDegrees=~
     AimFWDDeltaAngleDegrees=~
     AimBWDDeltaAngleDegrees=~
     AimLFTDeltaAngleDegrees=~
     AimRGTDeltaAngleDegrees=~
     SlopePitchDegreeAngle=~
     SlopeRollDegreeAngle=~
     SteerAngle=~
    ).SteerAngle
    BasePose:Pose Link Structure=Sequence Player 'Drive_Boost_Loop_GolfCart'() {sequence=&/Drive_Boost_Loop_GolfCart}
    Alpha:float=1
   ) -> (Pose:Pose Link Structure) {blendSpace=/Game/Animation/Game/MainPlayer/Vehicles/GolfCart/AO/DriveForwardLeans_AO_GolfCart, alphaBoolBlend {blendOption=Linear}}
   Output Animation Pose(Result=@n1)
  ]
 }
 Powersliding_Start {
  path=VehicleFinalPoseOverride.AnimGraphNode_StateMachine_0.Locomotion.AnimStateNode_16.Powersliding_Start
  nodes [
   n1 = "Blend Poses (EFortCustomGender)"!(
    BlendPose_0:Pose Link Structure=Sequence Player 'Braking_Powerslide_Start_GolfCart'() {bLoopAnimation=false, sequence=&/Braking_Powerslide_Start_GolfCart}
    BlendPose_1:Pose Link Structure=Sequence Player 'Braking_Powerslide_Start_CMF_GolfCart'() {bLoopAnimation=false, sequence=&/Braking_Powerslide_Start_CMF_GolfCart}
    BlendTime_0:float=0.1
    BlendTime_1:float=0.1
    ActiveEnumValue:Byte=Get Gender() -> (Gender:EFortCustomGender Enum=Invalid)
   ) -> (Pose:Pose Link Structure) {
    is=title AnimGraphNode_BlendListByEnum
    blendType=Linear
    blendPose [{linkId=-1, sourceLinkId=-1}, {linkId=-1, sourceLinkId=-1}]
    blendTime [0.1,0.1]
   }
   Output Animation Pose(Result=@n1)
  ]
 }
 Powersliding_Loop {
  path=VehicleFinalPoseOverride.AnimGraphNode_StateMachine_0.Locomotion.AnimStateNode_17.Powersliding_Loop
  nodes [
   n1 = AimOffset Player 'Driver_Powersliding_AO_Leans_GolfCart'!(
    X:float=Break Fort Anim Input Golf Cart(FortAnimInput_GolfCart=Get GolfCartInput()) -> (
     bIsUsingGolfCart=~
     bIsDriver=~
     bIsFrontPassenger=~
     bIsBackLeftPassenger=~
     bIsBackRightPassenger=~
     bIsFrontPassengerAndLeaning=~
     bIsBackPassengerAndLeaning=~
     bIsDrifting=~
     bIsBoosting=~
     bIsEBraking=~
     bIsReversing=~
     bIsBraking=~
     bIsMoving=~
     bIsMovingForward=~
     bIsPowerSliding=~
     bIsLeaning=~
     bIsLeaningOrBouncing=~
     bIsBounceCrouching=~
     bIsBounceCrouched=~
     bIsBounceJumping=~
     bIsBounceRecoiling=~
     bIsSteeringRight=~
     bIsSteeringLeft=~
     RunForwardAlpha=~
     BounceCompression=~
     LeanPosition=~
     LeanPositionX=~
     LeanPositionY=~
     LeanPositionZ=~
     bAimFWD=~
     bAimBWD=~
     bAimLFT=~
     bAimRGT=~
     PawnToVehicleDeltaYawAngleDegrees=~
     AimCardDirDeadZoneAngleDegrees=~
     AimCardDirAngleOffsetDegrees=~
     AimFWDDeltaAngleDegrees=~
     AimBWDDeltaAngleDegrees=~
     AimLFTDeltaAngleDegrees=~
     AimRGTDeltaAngleDegrees=~
     SlopePitchDegreeAngle=~
     SlopeRollDegreeAngle=~
     SteerAngle=~
    ).SteerAngle
    BasePose:Pose Link Structure=Sequence Player 'Braking_Powerslide_Loop_CMF_GolfCart'() {sequence=&/Braking_Powerslide_Loop_CMF_GolfCart}
    Alpha:float=1
   ) -> (Pose:Pose Link Structure) {blendSpace=/Game/Animation/Game/MainPlayer/Vehicles/GolfCart/AO/Driver_Powersliding_AO_Leans_GolfCart, alphaBoolBlend {blendOption=Linear}}
   n2 = AimOffset Player 'Driver_Powersliding_AO_Leans_GolfCart'(
    X=Break Fort Anim Input Golf Cart(FortAnimInput_GolfCart=Get GolfCartInput()) -> (
     bIsUsingGolfCart=~
     bIsDriver=~
     bIsFrontPassenger=~
     bIsBackLeftPassenger=~
     bIsBackRightPassenger=~
     bIsFrontPassengerAndLeaning=~
     bIsBackPassengerAndLeaning=~
     bIsDrifting=~
     bIsBoosting=~
     bIsEBraking=~
     bIsReversing=~
     bIsBraking=~
     bIsMoving=~
     bIsMovingForward=~
     bIsPowerSliding=~
     bIsLeaning=~
     bIsLeaningOrBouncing=~
     bIsBounceCrouching=~
     bIsBounceCrouched=~
     bIsBounceJumping=~
     bIsBounceRecoiling=~
     bIsSteeringRight=~
     bIsSteeringLeft=~
     RunForwardAlpha=~
     BounceCompression=~
     LeanPosition=~
     LeanPositionX=~
     LeanPositionY=~
     LeanPositionZ=~
     bAimFWD=~
     bAimBWD=~
     bAimLFT=~
     bAimRGT=~
     PawnToVehicleDeltaYawAngleDegrees=~
     AimCardDirDeadZoneAngleDegrees=~
     AimCardDirAngleOffsetDegrees=~
     AimFWDDeltaAngleDegrees=~
     AimBWDDeltaAngleDegrees=~
     AimLFTDeltaAngleDegrees=~
     AimRGTDeltaAngleDegrees=~
     SlopePitchDegreeAngle=~
     SlopeRollDegreeAngle=~
     SteerAngle=~
    ).SteerAngle
    BasePose=Sequence Player 'Braking_Powerslide_Loop_GolfCart'() {bLoopAnimation=false, sequence=&/Braking_Powerslide_Loop_GolfCart}
    Alpha=1
   ) {blendSpace=&Driver_Powersliding_AO_Leans_GolfCart, alphaBoolBlend {blendOption=Linear}}
   n3 = "Blend Poses (EFortCustomGender)"(BlendPose_0=@n2, BlendPose_1=@n1, BlendTime_0=0.1, BlendTime_1=0.1, ActiveEnumValue=Get Gender() -> (Gender=Invalid)) {blendType=Linear, blendPose [{linkId=-1, sourceLinkId=-1}, {linkId=-1, sourceLinkId=-1}], blendTime [0.1,0.1]}
   Output Animation Pose(Result=@n3)
  ]
 }
 Powersliding_End {
  path=VehicleFinalPoseOverride.AnimGraphNode_StateMachine_0.Locomotion.AnimStateNode_18.Powersliding_End
  nodes [
   n1 = "Blend Poses (EFortCustomGender)"(
    BlendPose_0=Sequence Player 'Braking_Powerslide_End_GolfCart'() {bLoopAnimation=false, sequence=&/Braking_Powerslide_End_GolfCart}
    BlendPose_1=Sequence Player 'Braking_Powerslide_End_CMF_GolfCart'() {bLoopAnimation=false, sequence=&/Braking_Powerslide_End_CMF_GolfCart}
    BlendTime_0=0.1
    BlendTime_1=0.1
    ActiveEnumValue=Get Gender() -> (Gender=Invalid)
   ) {blendType=Linear, blendPose [{linkId=-1, sourceLinkId=-1}, {linkId=-1, sourceLinkId=-1}], blendTime [0.1,0.1]}
   Output Animation Pose(Result=@n1)
  ]
 }
 Drive_N_Loop_GolfCart {
  path=VehicleFinalPoseOverride.AnimGraphNode_StateMachine_0.Locomotion.AnimStateNode_2.Drive_N_Loop_GolfCart
  nodes [Output Animation Pose(Result=Apply Mesh Space Additive(
     Base=Use cached pose 'Idle_Pose'()
     Additive="Sequence Player (additive) 'Drive_N_Pose_Additive_GolfCart'"() {sequence=&/Drive_N_Pose_Additive_GolfCart}
     Alpha=1
    ) {alphaBoolBlend {blendOption=Linear}})]
 }
 Reverse {
  path=VehicleFinalPoseOverride.AnimGraphNode_StateMachine_0.Locomotion.AnimStateNode_3.Reverse
  nodes [Output Animation Pose(Result=Sequence Player 'Reverse_Center_Additive_GolfCart'() {sequence=&/Reverse_Center_Additive_GolfCart})]
 }
 Braking {
  path=VehicleFinalPoseOverride.AnimGraphNode_StateMachine_0.Locomotion.AnimStateNode_4.Braking
  nodes [Output Animation Pose(Result=Sequence Player 'Braking_GolfCart'() {bLoopAnimation=false, sequence=&/Braking_GolfCart})]
 }
 Back_to_Idle_Transition {
  path=VehicleFinalPoseOverride.AnimGraphNode_StateMachine_0.Locomotion.AnimStateNode_5.Back_to_Idle_Transition
  nodes [Output Animation Pose(Result=Sequence Player 'Drive_N_BackToIdle_GolfCart'() {bLoopAnimation=false, sequence=&/Drive_N_BackToIdle_GolfCart})]
 }
 Reverse_Start {
  path=VehicleFinalPoseOverride.AnimGraphNode_StateMachine_0.Locomotion.AnimStateNode_6.Reverse_Start
  nodes [
   n1 = Sequence Player 'Reverse_Start_GolfCart'() {bLoopAnimation=false, playRate=1.2, sequence=&/Reverse_Start_GolfCart, startPosition=0.1}
   Output Animation Pose(Result=@n1)
  ]
 }
 Boost_Start {
  path=VehicleFinalPoseOverride.AnimGraphNode_StateMachine_0.Locomotion.AnimStateNode_7.Boost_Start
  nodes [Output Animation Pose(Result=Sequence Player 'Drive_Boost_GolfCart'() {bLoopAnimation=false, sequence=&/Drive_Boost_GolfCart})]
 }
 Reverse_End {
  path=VehicleFinalPoseOverride.AnimGraphNode_StateMachine_0.Locomotion.AnimStateNode_8.Reverse_End
  nodes [Output Animation Pose(Result=Sequence Player 'Reverse_End_GolfCart'() {bLoopAnimation=false, sequence=&/Reverse_End_GolfCart})]
 }
 Transition {
  path=VehicleFinalPoseOverride.AnimGraphNode_StateMachine_0.Locomotion.AnimStateTransitionNode_0.Transition
  nodes [Result(bCanEnterTransition=Break Fort Anim Input Golf Cart(FortAnimInput_GolfCart=Get GolfCartInput()) -> (
     bIsUsingGolfCart=~
     bIsDriver=~
     bIsFrontPassenger=~
     bIsBackLeftPassenger=~
     bIsBackRightPassenger=~
     bIsFrontPassengerAndLeaning=~
     bIsBackPassengerAndLeaning=~
     bIsDrifting=~
     bIsBoosting=~
     bIsEBraking=~
     bIsReversing=~
     bIsBraking=~
     bIsMoving=~
     bIsMovingForward=~
     bIsPowerSliding=~
     bIsLeaning=~
     bIsLeaningOrBouncing=~
     bIsBounceCrouching=~
     bIsBounceCrouched=~
     bIsBounceJumping=~
     bIsBounceRecoiling=~
     bIsSteeringRight=~
     bIsSteeringLeft=~
     RunForwardAlpha=~
     BounceCompression=~
     LeanPosition=~
     LeanPositionX=~
     LeanPositionY=~
     LeanPositionZ=~
     bAimFWD=~
     bAimBWD=~
     bAimLFT=~
     bAimRGT=~
     PawnToVehicleDeltaYawAngleDegrees=~
     AimCardDirDeadZoneAngleDegrees=~
     AimCardDirAngleOffsetDegrees=~
     AimFWDDeltaAngleDegrees=~
     AimBWDDeltaAngleDegrees=~
     AimLFTDeltaAngleDegrees=~
     AimRGTDeltaAngleDegrees=~
     SlopePitchDegreeAngle=~
     SlopeRollDegreeAngle=~
     SteerAngle=~
    ).bIsMovingForward)]
 }
 Transition {path=VehicleFinalPoseOverride.AnimGraphNode_StateMachine_0.Locomotion.AnimStateTransitionNode_1.Transition, nodes [Result()]}
 Transition {path=VehicleFinalPoseOverride.AnimGraphNode_StateMachine_0.Locomotion.AnimStateTransitionNode_10.Transition, nodes [Result()]}
 Transition {
  path=VehicleFinalPoseOverride.AnimGraphNode_StateMachine_0.Locomotion.AnimStateTransitionNode_11.Transition
  nodes [Result(bCanEnterTransition=Break Fort Anim Input Golf Cart(FortAnimInput_GolfCart=Get GolfCartInput()) -> (
     bIsUsingGolfCart=~
     bIsDriver=~
     bIsFrontPassenger=~
     bIsBackLeftPassenger=~
     bIsBackRightPassenger=~
     bIsFrontPassengerAndLeaning=~
     bIsBackPassengerAndLeaning=~
     bIsDrifting=~
     bIsBoosting=~
     bIsEBraking=~
     bIsReversing=~
     bIsBraking=~
     bIsMoving=~
     bIsMovingForward=~
     bIsPowerSliding=~
     bIsLeaning=~
     bIsLeaningOrBouncing=~
     bIsBounceCrouching=~
     bIsBounceCrouched=~
     bIsBounceJumping=~
     bIsBounceRecoiling=~
     bIsSteeringRight=~
     bIsSteeringLeft=~
     RunForwardAlpha=~
     BounceCompression=~
     LeanPosition=~
     LeanPositionX=~
     LeanPositionY=~
     LeanPositionZ=~
     bAimFWD=~
     bAimBWD=~
     bAimLFT=~
     bAimRGT=~
     PawnToVehicleDeltaYawAngleDegrees=~
     AimCardDirDeadZoneAngleDegrees=~
     AimCardDirAngleOffsetDegrees=~
     AimFWDDeltaAngleDegrees=~
     AimBWDDeltaAngleDegrees=~
     AimLFTDeltaAngleDegrees=~
     AimRGTDeltaAngleDegrees=~
     SlopePitchDegreeAngle=~
     SlopeRollDegreeAngle=~
     SteerAngle=~
    ).bIsReversing)]
 }
 Transition {path=VehicleFinalPoseOverride.AnimGraphNode_StateMachine_0.Locomotion.AnimStateTransitionNode_12.Transition, nodes [Result()]}
 Transition {path=VehicleFinalPoseOverride.AnimGraphNode_StateMachine_0.Locomotion.AnimStateTransitionNode_13.Transition, nodes [Result()]}
 Transition {
  path=VehicleFinalPoseOverride.AnimGraphNode_StateMachine_0.Locomotion.AnimStateTransitionNode_14.Transition
  nodes [Result(bCanEnterTransition=Break Fort Anim Input Golf Cart(FortAnimInput_GolfCart=Get GolfCartInput()) -> (
     bIsUsingGolfCart=~
     bIsDriver=~
     bIsFrontPassenger=~
     bIsBackLeftPassenger=~
     bIsBackRightPassenger=~
     bIsFrontPassengerAndLeaning=~
     bIsBackPassengerAndLeaning=~
     bIsDrifting=~
     bIsBoosting=~
     bIsEBraking=~
     bIsReversing=~
     bIsBraking=~
     bIsMoving=~
     bIsMovingForward=~
     bIsPowerSliding=~
     bIsLeaning=~
     bIsLeaningOrBouncing=~
     bIsBounceCrouching=~
     bIsBounceCrouched=~
     bIsBounceJumping=~
     bIsBounceRecoiling=~
     bIsSteeringRight=~
     bIsSteeringLeft=~
     RunForwardAlpha=~
     BounceCompression=~
     LeanPosition=~
     LeanPositionX=~
     LeanPositionY=~
     LeanPositionZ=~
     bAimFWD=~
     bAimBWD=~
     bAimLFT=~
     bAimRGT=~
     PawnToVehicleDeltaYawAngleDegrees=~
     AimCardDirDeadZoneAngleDegrees=~
     AimCardDirAngleOffsetDegrees=~
     AimFWDDeltaAngleDegrees=~
     AimBWDDeltaAngleDegrees=~
     AimLFTDeltaAngleDegrees=~
     AimRGTDeltaAngleDegrees=~
     SlopePitchDegreeAngle=~
     SlopeRollDegreeAngle=~
     SteerAngle=~
    ).bIsPowerSliding)]
 }
 Transition {
  path=VehicleFinalPoseOverride.AnimGraphNode_StateMachine_0.Locomotion.AnimStateTransitionNode_15.Transition
  nodes [Result(bCanEnterTransition=Break Fort Anim Input Golf Cart(FortAnimInput_GolfCart=Get GolfCartInput()) -> (
     bIsUsingGolfCart=~
     bIsDriver=~
     bIsFrontPassenger=~
     bIsBackLeftPassenger=~
     bIsBackRightPassenger=~
     bIsFrontPassengerAndLeaning=~
     bIsBackPassengerAndLeaning=~
     bIsDrifting=~
     bIsBoosting=~
     bIsEBraking=~
     bIsReversing=~
     bIsBraking=~
     bIsMoving=~
     bIsMovingForward=~
     bIsPowerSliding=~
     bIsLeaning=~
     bIsLeaningOrBouncing=~
     bIsBounceCrouching=~
     bIsBounceCrouched=~
     bIsBounceJumping=~
     bIsBounceRecoiling=~
     bIsSteeringRight=~
     bIsSteeringLeft=~
     RunForwardAlpha=~
     BounceCompression=~
     LeanPosition=~
     LeanPositionX=~
     LeanPositionY=~
     LeanPositionZ=~
     bAimFWD=~
     bAimBWD=~
     bAimLFT=~
     bAimRGT=~
     PawnToVehicleDeltaYawAngleDegrees=~
     AimCardDirDeadZoneAngleDegrees=~
     AimCardDirAngleOffsetDegrees=~
     AimFWDDeltaAngleDegrees=~
     AimBWDDeltaAngleDegrees=~
     AimLFTDeltaAngleDegrees=~
     AimRGTDeltaAngleDegrees=~
     SlopePitchDegreeAngle=~
     SlopeRollDegreeAngle=~
     SteerAngle=~
    ).bIsBoosting)]
 }
 Transition {
  path=VehicleFinalPoseOverride.AnimGraphNode_StateMachine_0.Locomotion.AnimStateTransitionNode_16.Transition
  nodes [
   n1 = Get GolfCartInput()
   Result(bCanEnterTransition=NOT Boolean(A=Break Fort Anim Input Golf Cart(FortAnimInput_GolfCart=@n1) -> (
      bIsUsingGolfCart=~
      bIsDriver=~
      bIsFrontPassenger=~
      bIsBackLeftPassenger=~
      bIsBackRightPassenger=~
      bIsFrontPassengerAndLeaning=~
      bIsBackPassengerAndLeaning=~
      bIsDrifting=~
      bIsBoosting=~
      bIsEBraking=~
      bIsReversing=~
      bIsBraking=~
      bIsMoving=~
      bIsMovingForward=~
      bIsPowerSliding=~
      bIsLeaning=~
      bIsLeaningOrBouncing=~
      bIsBounceCrouching=~
      bIsBounceCrouched=~
      bIsBounceJumping=~
      bIsBounceRecoiling=~
      bIsSteeringRight=~
      bIsSteeringLeft=~
      RunForwardAlpha=~
      BounceCompression=~
      LeanPosition=~
      LeanPositionX=~
      LeanPositionY=~
      LeanPositionZ=~
      bAimFWD=~
      bAimBWD=~
      bAimLFT=~
      bAimRGT=~
      PawnToVehicleDeltaYawAngleDegrees=~
      AimCardDirDeadZoneAngleDegrees=~
      AimCardDirAngleOffsetDegrees=~
      AimFWDDeltaAngleDegrees=~
      AimBWDDeltaAngleDegrees=~
      AimLFTDeltaAngleDegrees=~
      AimRGTDeltaAngleDegrees=~
      SlopePitchDegreeAngle=~
      SlopeRollDegreeAngle=~
      SteerAngle=~
     ).bIsEBraking))
  ]
 }
 Transition {path=VehicleFinalPoseOverride.AnimGraphNode_StateMachine_0.Locomotion.AnimStateTransitionNode_17.Transition, nodes [Result()]}
 Transition {
  path=VehicleFinalPoseOverride.AnimGraphNode_StateMachine_0.Locomotion.AnimStateTransitionNode_18.Transition
  nodes [Result(bCanEnterTransition=Break Fort Anim Input Golf Cart(FortAnimInput_GolfCart=Get GolfCartInput()) -> (
     bIsUsingGolfCart=~
     bIsDriver=~
     bIsFrontPassenger=~
     bIsBackLeftPassenger=~
     bIsBackRightPassenger=~
     bIsFrontPassengerAndLeaning=~
     bIsBackPassengerAndLeaning=~
     bIsDrifting=~
     bIsBoosting=~
     bIsEBraking=~
     bIsReversing=~
     bIsBraking=~
     bIsMoving=~
     bIsMovingForward=~
     bIsPowerSliding=~
     bIsLeaning=~
     bIsLeaningOrBouncing=~
     bIsBounceCrouching=~
     bIsBounceCrouched=~
     bIsBounceJumping=~
     bIsBounceRecoiling=~
     bIsSteeringRight=~
     bIsSteeringLeft=~
     RunForwardAlpha=~
     BounceCompression=~
     LeanPosition=~
     LeanPositionX=~
     LeanPositionY=~
     LeanPositionZ=~
     bAimFWD=~
     bAimBWD=~
     bAimLFT=~
     bAimRGT=~
     PawnToVehicleDeltaYawAngleDegrees=~
     AimCardDirDeadZoneAngleDegrees=~
     AimCardDirAngleOffsetDegrees=~
     AimFWDDeltaAngleDegrees=~
     AimBWDDeltaAngleDegrees=~
     AimLFTDeltaAngleDegrees=~
     AimRGTDeltaAngleDegrees=~
     SlopePitchDegreeAngle=~
     SlopeRollDegreeAngle=~
     SteerAngle=~
    ).bIsEBraking)]
 }
 Transition {
  path=VehicleFinalPoseOverride.AnimGraphNode_StateMachine_0.Locomotion.AnimStateTransitionNode_19.Transition
  nodes [Result(bCanEnterTransition=Break Fort Anim Input Golf Cart(FortAnimInput_GolfCart=Get GolfCartInput()) -> (
     bIsUsingGolfCart=~
     bIsDriver=~
     bIsFrontPassenger=~
     bIsBackLeftPassenger=~
     bIsBackRightPassenger=~
     bIsFrontPassengerAndLeaning=~
     bIsBackPassengerAndLeaning=~
     bIsDrifting=~
     bIsBoosting=~
     bIsEBraking=~
     bIsReversing=~
     bIsBraking=~
     bIsMoving=~
     bIsMovingForward=~
     bIsPowerSliding=~
     bIsLeaning=~
     bIsLeaningOrBouncing=~
     bIsBounceCrouching=~
     bIsBounceCrouched=~
     bIsBounceJumping=~
     bIsBounceRecoiling=~
     bIsSteeringRight=~
     bIsSteeringLeft=~
     RunForwardAlpha=~
     BounceCompression=~
     LeanPosition=~
     LeanPositionX=~
     LeanPositionY=~
     LeanPositionZ=~
     bAimFWD=~
     bAimBWD=~
     bAimLFT=~
     bAimRGT=~
     PawnToVehicleDeltaYawAngleDegrees=~
     AimCardDirDeadZoneAngleDegrees=~
     AimCardDirAngleOffsetDegrees=~
     AimFWDDeltaAngleDegrees=~
     AimBWDDeltaAngleDegrees=~
     AimLFTDeltaAngleDegrees=~
     AimRGTDeltaAngleDegrees=~
     SlopePitchDegreeAngle=~
     SlopeRollDegreeAngle=~
     SteerAngle=~
    ).bIsEBraking)]
 }
 Transition {
  path=VehicleFinalPoseOverride.AnimGraphNode_StateMachine_0.Locomotion.AnimStateTransitionNode_2.Transition
  nodes [Result(bCanEnterTransition=Break Fort Anim Input Golf Cart(FortAnimInput_GolfCart=Get GolfCartInput()) -> (
     bIsUsingGolfCart=~
     bIsDriver=~
     bIsFrontPassenger=~
     bIsBackLeftPassenger=~
     bIsBackRightPassenger=~
     bIsFrontPassengerAndLeaning=~
     bIsBackPassengerAndLeaning=~
     bIsDrifting=~
     bIsBoosting=~
     bIsEBraking=~
     bIsReversing=~
     bIsBraking=~
     bIsMoving=~
     bIsMovingForward=~
     bIsPowerSliding=~
     bIsLeaning=~
     bIsLeaningOrBouncing=~
     bIsBounceCrouching=~
     bIsBounceCrouched=~
     bIsBounceJumping=~
     bIsBounceRecoiling=~
     bIsSteeringRight=~
     bIsSteeringLeft=~
     RunForwardAlpha=~
     BounceCompression=~
     LeanPosition=~
     LeanPositionX=~
     LeanPositionY=~
     LeanPositionZ=~
     bAimFWD=~
     bAimBWD=~
     bAimLFT=~
     bAimRGT=~
     PawnToVehicleDeltaYawAngleDegrees=~
     AimCardDirDeadZoneAngleDegrees=~
     AimCardDirAngleOffsetDegrees=~
     AimFWDDeltaAngleDegrees=~
     AimBWDDeltaAngleDegrees=~
     AimLFTDeltaAngleDegrees=~
     AimRGTDeltaAngleDegrees=~
     SlopePitchDegreeAngle=~
     SlopeRollDegreeAngle=~
     SteerAngle=~
    ).bIsReversing)]
 }
 Transition {path=VehicleFinalPoseOverride.AnimGraphNode_StateMachine_0.Locomotion.AnimStateTransitionNode_20.Transition, nodes [Result()]}
 Transition {
  path=VehicleFinalPoseOverride.AnimGraphNode_StateMachine_0.Locomotion.AnimStateTransitionNode_21.Transition
  nodes [
   n1 = Get GolfCartInput()
   Result(bCanEnterTransition=NOT Boolean(A=Break Fort Anim Input Golf Cart(FortAnimInput_GolfCart=@n1) -> (
      bIsUsingGolfCart=~
      bIsDriver=~
      bIsFrontPassenger=~
      bIsBackLeftPassenger=~
      bIsBackRightPassenger=~
      bIsFrontPassengerAndLeaning=~
      bIsBackPassengerAndLeaning=~
      bIsDrifting=~
      bIsBoosting=~
      bIsEBraking=~
      bIsReversing=~
      bIsBraking=~
      bIsMoving=~
      bIsMovingForward=~
      bIsPowerSliding=~
      bIsLeaning=~
      bIsLeaningOrBouncing=~
      bIsBounceCrouching=~
      bIsBounceCrouched=~
      bIsBounceJumping=~
      bIsBounceRecoiling=~
      bIsSteeringRight=~
      bIsSteeringLeft=~
      RunForwardAlpha=~
      BounceCompression=~
      LeanPosition=~
      LeanPositionX=~
      LeanPositionY=~
      LeanPositionZ=~
      bAimFWD=~
      bAimBWD=~
      bAimLFT=~
      bAimRGT=~
      PawnToVehicleDeltaYawAngleDegrees=~
      AimCardDirDeadZoneAngleDegrees=~
      AimCardDirAngleOffsetDegrees=~
      AimFWDDeltaAngleDegrees=~
      AimBWDDeltaAngleDegrees=~
      AimLFTDeltaAngleDegrees=~
      AimRGTDeltaAngleDegrees=~
      SlopePitchDegreeAngle=~
      SlopeRollDegreeAngle=~
      SteerAngle=~
     ).bIsBoosting))
  ]
 }
 Transition {
  path=VehicleFinalPoseOverride.AnimGraphNode_StateMachine_0.Locomotion.AnimStateTransitionNode_23.Transition
  nodes [Result(bCanEnterTransition=Break Fort Anim Input Golf Cart(FortAnimInput_GolfCart=Get GolfCartInput()) -> (
     bIsUsingGolfCart=~
     bIsDriver=~
     bIsFrontPassenger=~
     bIsBackLeftPassenger=~
     bIsBackRightPassenger=~
     bIsFrontPassengerAndLeaning=~
     bIsBackPassengerAndLeaning=~
     bIsDrifting=~
     bIsBoosting=~
     bIsEBraking=~
     bIsReversing=~
     bIsBraking=~
     bIsMoving=~
     bIsMovingForward=~
     bIsPowerSliding=~
     bIsLeaning=~
     bIsLeaningOrBouncing=~
     bIsBounceCrouching=~
     bIsBounceCrouched=~
     bIsBounceJumping=~
     bIsBounceRecoiling=~
     bIsSteeringRight=~
     bIsSteeringLeft=~
     RunForwardAlpha=~
     BounceCompression=~
     LeanPosition=~
     LeanPositionX=~
     LeanPositionY=~
     LeanPositionZ=~
     bAimFWD=~
     bAimBWD=~
     bAimLFT=~
     bAimRGT=~
     PawnToVehicleDeltaYawAngleDegrees=~
     AimCardDirDeadZoneAngleDegrees=~
     AimCardDirAngleOffsetDegrees=~
     AimFWDDeltaAngleDegrees=~
     AimBWDDeltaAngleDegrees=~
     AimLFTDeltaAngleDegrees=~
     AimRGTDeltaAngleDegrees=~
     SlopePitchDegreeAngle=~
     SlopeRollDegreeAngle=~
     SteerAngle=~
    ).bIsEBraking)]
 }
 Transition {
  path=VehicleFinalPoseOverride.AnimGraphNode_StateMachine_0.Locomotion.AnimStateTransitionNode_3.Transition
  nodes [
   n1 = Get GolfCartInput()
   Result(bCanEnterTransition=NOT Boolean(A=Break Fort Anim Input Golf Cart(FortAnimInput_GolfCart=@n1) -> (
      bIsUsingGolfCart=~
      bIsDriver=~
      bIsFrontPassenger=~
      bIsBackLeftPassenger=~
      bIsBackRightPassenger=~
      bIsFrontPassengerAndLeaning=~
      bIsBackPassengerAndLeaning=~
      bIsDrifting=~
      bIsBoosting=~
      bIsEBraking=~
      bIsReversing=~
      bIsBraking=~
      bIsMoving=~
      bIsMovingForward=~
      bIsPowerSliding=~
      bIsLeaning=~
      bIsLeaningOrBouncing=~
      bIsBounceCrouching=~
      bIsBounceCrouched=~
      bIsBounceJumping=~
      bIsBounceRecoiling=~
      bIsSteeringRight=~
      bIsSteeringLeft=~
      RunForwardAlpha=~
      BounceCompression=~
      LeanPosition=~
      LeanPositionX=~
      LeanPositionY=~
      LeanPositionZ=~
      bAimFWD=~
      bAimBWD=~
      bAimLFT=~
      bAimRGT=~
      PawnToVehicleDeltaYawAngleDegrees=~
      AimCardDirDeadZoneAngleDegrees=~
      AimCardDirAngleOffsetDegrees=~
      AimFWDDeltaAngleDegrees=~
      AimBWDDeltaAngleDegrees=~
      AimLFTDeltaAngleDegrees=~
      AimRGTDeltaAngleDegrees=~
      SlopePitchDegreeAngle=~
      SlopeRollDegreeAngle=~
      SteerAngle=~
     ).bIsMoving))
  ]
 }
 Transition {path=VehicleFinalPoseOverride.AnimGraphNode_StateMachine_0.Locomotion.AnimStateTransitionNode_4.Transition, nodes [Result()]}
 Transition {
  path=VehicleFinalPoseOverride.AnimGraphNode_StateMachine_0.Locomotion.AnimStateTransitionNode_5.Transition
  nodes [Result(bCanEnterTransition=Break Fort Anim Input Golf Cart(FortAnimInput_GolfCart=Get GolfCartInput()) -> (
     bIsUsingGolfCart=~
     bIsDriver=~
     bIsFrontPassenger=~
     bIsBackLeftPassenger=~
     bIsBackRightPassenger=~
     bIsFrontPassengerAndLeaning=~
     bIsBackPassengerAndLeaning=~
     bIsDrifting=~
     bIsBoosting=~
     bIsEBraking=~
     bIsReversing=~
     bIsBraking=~
     bIsMoving=~
     bIsMovingForward=~
     bIsPowerSliding=~
     bIsLeaning=~
     bIsLeaningOrBouncing=~
     bIsBounceCrouching=~
     bIsBounceCrouched=~
     bIsBounceJumping=~
     bIsBounceRecoiling=~
     bIsSteeringRight=~
     bIsSteeringLeft=~
     RunForwardAlpha=~
     BounceCompression=~
     LeanPosition=~
     LeanPositionX=~
     LeanPositionY=~
     LeanPositionZ=~
     bAimFWD=~
     bAimBWD=~
     bAimLFT=~
     bAimRGT=~
     PawnToVehicleDeltaYawAngleDegrees=~
     AimCardDirDeadZoneAngleDegrees=~
     AimCardDirAngleOffsetDegrees=~
     AimFWDDeltaAngleDegrees=~
     AimBWDDeltaAngleDegrees=~
     AimLFTDeltaAngleDegrees=~
     AimRGTDeltaAngleDegrees=~
     SlopePitchDegreeAngle=~
     SlopeRollDegreeAngle=~
     SteerAngle=~
    ).bIsBraking)]
 }
 Transition {
  path=VehicleFinalPoseOverride.AnimGraphNode_StateMachine_0.Locomotion.AnimStateTransitionNode_6.Transition
  nodes [Result(bCanEnterTransition=Break Fort Anim Input Golf Cart(FortAnimInput_GolfCart=Get GolfCartInput()) -> (
     bIsUsingGolfCart=~
     bIsDriver=~
     bIsFrontPassenger=~
     bIsBackLeftPassenger=~
     bIsBackRightPassenger=~
     bIsFrontPassengerAndLeaning=~
     bIsBackPassengerAndLeaning=~
     bIsDrifting=~
     bIsBoosting=~
     bIsEBraking=~
     bIsReversing=~
     bIsBraking=~
     bIsMoving=~
     bIsMovingForward=~
     bIsPowerSliding=~
     bIsLeaning=~
     bIsLeaningOrBouncing=~
     bIsBounceCrouching=~
     bIsBounceCrouched=~
     bIsBounceJumping=~
     bIsBounceRecoiling=~
     bIsSteeringRight=~
     bIsSteeringLeft=~
     RunForwardAlpha=~
     BounceCompression=~
     LeanPosition=~
     LeanPositionX=~
     LeanPositionY=~
     LeanPositionZ=~
     bAimFWD=~
     bAimBWD=~
     bAimLFT=~
     bAimRGT=~
     PawnToVehicleDeltaYawAngleDegrees=~
     AimCardDirDeadZoneAngleDegrees=~
     AimCardDirAngleOffsetDegrees=~
     AimFWDDeltaAngleDegrees=~
     AimBWDDeltaAngleDegrees=~
     AimLFTDeltaAngleDegrees=~
     AimRGTDeltaAngleDegrees=~
     SlopePitchDegreeAngle=~
     SlopeRollDegreeAngle=~
     SteerAngle=~
    ).bIsPowerSliding)]
 }
 Transition {
  path=VehicleFinalPoseOverride.AnimGraphNode_StateMachine_0.Locomotion.AnimStateTransitionNode_7.Transition
  nodes [Result(bCanEnterTransition=Break Fort Anim Input Golf Cart(FortAnimInput_GolfCart=Get GolfCartInput()) -> (
     bIsUsingGolfCart=~
     bIsDriver=~
     bIsFrontPassenger=~
     bIsBackLeftPassenger=~
     bIsBackRightPassenger=~
     bIsFrontPassengerAndLeaning=~
     bIsBackPassengerAndLeaning=~
     bIsDrifting=~
     bIsBoosting=~
     bIsEBraking=~
     bIsReversing=~
     bIsBraking=~
     bIsMoving=~
     bIsMovingForward=~
     bIsPowerSliding=~
     bIsLeaning=~
     bIsLeaningOrBouncing=~
     bIsBounceCrouching=~
     bIsBounceCrouched=~
     bIsBounceJumping=~
     bIsBounceRecoiling=~
     bIsSteeringRight=~
     bIsSteeringLeft=~
     RunForwardAlpha=~
     BounceCompression=~
     LeanPosition=~
     LeanPositionX=~
     LeanPositionY=~
     LeanPositionZ=~
     bAimFWD=~
     bAimBWD=~
     bAimLFT=~
     bAimRGT=~
     PawnToVehicleDeltaYawAngleDegrees=~
     AimCardDirDeadZoneAngleDegrees=~
     AimCardDirAngleOffsetDegrees=~
     AimFWDDeltaAngleDegrees=~
     AimBWDDeltaAngleDegrees=~
     AimLFTDeltaAngleDegrees=~
     AimRGTDeltaAngleDegrees=~
     SlopePitchDegreeAngle=~
     SlopeRollDegreeAngle=~
     SteerAngle=~
    ).bIsMovingForward)]
 }
 Transition {
  path=VehicleFinalPoseOverride.AnimGraphNode_StateMachine_0.Locomotion.AnimStateTransitionNode_8.Transition
  nodes [
   n1 = Get GolfCartInput()
   Result(bCanEnterTransition=NOT Boolean(A=Break Fort Anim Input Golf Cart(FortAnimInput_GolfCart=@n1) -> (
      bIsUsingGolfCart=~
      bIsDriver=~
      bIsFrontPassenger=~
      bIsBackLeftPassenger=~
      bIsBackRightPassenger=~
      bIsFrontPassengerAndLeaning=~
      bIsBackPassengerAndLeaning=~
      bIsDrifting=~
      bIsBoosting=~
      bIsEBraking=~
      bIsReversing=~
      bIsBraking=~
      bIsMoving=~
      bIsMovingForward=~
      bIsPowerSliding=~
      bIsLeaning=~
      bIsLeaningOrBouncing=~
      bIsBounceCrouching=~
      bIsBounceCrouched=~
      bIsBounceJumping=~
      bIsBounceRecoiling=~
      bIsSteeringRight=~
      bIsSteeringLeft=~
      RunForwardAlpha=~
      BounceCompression=~
      LeanPosition=~
      LeanPositionX=~
      LeanPositionY=~
      LeanPositionZ=~
      bAimFWD=~
      bAimBWD=~
      bAimLFT=~
      bAimRGT=~
      PawnToVehicleDeltaYawAngleDegrees=~
      AimCardDirDeadZoneAngleDegrees=~
      AimCardDirAngleOffsetDegrees=~
      AimFWDDeltaAngleDegrees=~
      AimBWDDeltaAngleDegrees=~
      AimLFTDeltaAngleDegrees=~
      AimRGTDeltaAngleDegrees=~
      SlopePitchDegreeAngle=~
      SlopeRollDegreeAngle=~
      SteerAngle=~
     ).bIsMoving))
  ]
 }
 Transition {path=VehicleFinalPoseOverride.AnimGraphNode_StateMachine_0.Locomotion.AnimStateTransitionNode_9.Transition, nodes [Result()]}
 Driver_Idle {
  entry=Idle_Leans
  path=VehicleFinalPoseOverride.AnimGraphNode_StateMachine_1.Driver_Idle
  ids=id
  nodes [AnimStateNode_0 = Idle_Leans() {is=title AnimStateNode}]
 }
 Idle_Leans {
  path=VehicleFinalPoseOverride.AnimGraphNode_StateMachine_1.Driver_Idle.AnimStateNode_0.Idle_Leans
  nodes [
   n1 = Break Fort Anim Input Golf Cart(FortAnimInput_GolfCart=Get GolfCartInput()) -> (
    bIsUsingGolfCart=~
    bIsDriver=~
    bIsFrontPassenger=~
    bIsBackLeftPassenger=~
    bIsBackRightPassenger=~
    bIsFrontPassengerAndLeaning=~
    bIsBackPassengerAndLeaning=~
    bIsDrifting=~
    bIsBoosting=~
    bIsEBraking=~
    bIsReversing=~
    bIsBraking=~
    bIsMoving=~
    bIsMovingForward=~
    bIsPowerSliding=~
    bIsLeaning=~
    bIsLeaningOrBouncing=~
    bIsBounceCrouching=~
    bIsBounceCrouched=~
    bIsBounceJumping=~
    bIsBounceRecoiling=~
    bIsSteeringRight=~
    bIsSteeringLeft=~
    RunForwardAlpha=~
    BounceCompression=~
    LeanPosition=~
    LeanPositionX=~
    LeanPositionY=~
    LeanPositionZ=~
    bAimFWD=~
    bAimBWD=~
    bAimLFT=~
    bAimRGT=~
    PawnToVehicleDeltaYawAngleDegrees=~
    AimCardDirDeadZoneAngleDegrees=~
    AimCardDirAngleOffsetDegrees=~
    AimFWDDeltaAngleDegrees=~
    AimBWDDeltaAngleDegrees=~
    AimLFTDeltaAngleDegrees=~
    AimRGTDeltaAngleDegrees=~
    SlopePitchDegreeAngle=~
    SlopeRollDegreeAngle=~
    SteerAngle=~
   )
   n2 = AimOffset Player 'DriveForwardLeans_AO_GolfCart'(
    X=@n1.SteerAngle
    BasePose=Sequence Player 'Idle_Pose_Driver_GolfCart'() {sequence=&/Idle_Pose_Driver_GolfCart}
    Alpha=1
   ) {blendSpace=&DriveForwardLeans_AO_GolfCart, alphaBoolBlend {blendOption=Linear}}
   n3 = Blend Poses by bool(
    BlendPose_0=Sequence Player 'Drive_Right_Idle_GolfCart'() {sequence=&/Drive_Right_Idle_GolfCart}
    BlendPose_1=@n2
    BlendTime_0=0.35
    BlendTime_1=0.35
    bActiveValue=@n1.bIsSteeringRight
   ) {blendType=Linear}
   n4 = Sequence Player 'Drive_Left_Idle_GolfCart'() {sequence=&/Drive_Left_Idle_GolfCart}
   n5 = AimOffset Player 'DriveForwardLeans_AO_GolfCart'(
    X=@n1.SteerAngle
    BasePose=Sequence Player 'Idle_Pose_Driver_GolfCart'() {sequence=&Idle_Pose_Driver_GolfCart}
    Alpha=1
   ) {blendSpace=&DriveForwardLeans_AO_GolfCart, alphaBoolBlend {blendOption=Linear}}
   Output Animation Pose(Result=Blend Poses by bool(
     BlendPose_0=Blend Poses by bool(BlendPose_0=@n4, BlendPose_1=@n3, BlendTime_0=0.35, BlendTime_1=0.35, bActiveValue=@n1.bIsSteeringLeft) {blendType=Linear}
     BlendPose_1=@n5
     BlendTime_0=0.35
     BlendTime_1=0.35
     bActiveValue=@n1.bIsMoving
    ) {blendType=Linear})
  ]
 }
 VehicleFullBodyOverride {nodes [
   Input Pose() {name=InPoseFullBody}
   Output Pose() {layerGroup=VehicleOverride}
  ]}
 VehicleLowerBodyOverride {nodes [
   Input Pose() {name=InPoseLowerBodyDefault}
   Input Pose() {name=InPoseUpperBody}
   Output Pose() {layerGroup=VehicleOverride}
  ]}
 VehicleSplitBodyOverride {nodes [
   Input Pose() {name=InPoseSplitBody}
   Input Pose() {name=InPoseUpperAndLowerBody}
   Input Pose() {name=InPoseUpperAndLowerBodyRemoveRoot}
   Output Pose() {layerGroup=Vehicle Override}
  ]}
]
variables [OnPedal {default=()}]
